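# Physics asset for the female medium Elf body's skirt, simulating frill panels around the pelvis and thighs.
# Box bodies on dyn_frill bones hang by angle-limited constraints from small driver spheres, which attach with free rotation to the pelvis or sub-thigh bodies.
# Extra constraints tie the back center panel to its neighbors and each side panel to its front neighbor.
# Pelvis and thigh capsules provide leg volume; only the left thigh-to-pelvis joint is angle-limited.
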PhysicsAsset /BRCosmetics/Characters/Player/Female/Medium/Bodies/F_MED_Elf/Meshes/F_MED_Elf_Skirt_R_Physics
bodies {
 Flap_Front_Driven_Bone [{shapeType=Sphere, radius=0.505}]
 driver_frill_bck [{shapeType=Sphere, radius=0.505}]
 driver_frill_bck_l [{shapeType=Sphere, radius=0.505}]
 driver_frill_bck_r [{shapeType=Sphere, radius=0.505}]
 driver_frill_frt_l [{shapeType=Sphere, radius=1}]
 driver_frill_frt_r [{shapeType=Sphere, radius=0.505}]
 driver_frill_side_l [{shapeType=Sphere, radius=0.505}]
 driver_frill_side_r [{shapeType=Sphere, radius=0.505}]
 dyn_frill_bck [{shapeType=Box, extentX=15, extentY=19, extentZ=7.5, center=<0 -0.4986 -9.4894>, rotation=<rot 0 0 99.0703>}]
 dyn_frill_bck_l [{shapeType=Box, extentX=19, extentY=14.5, extentZ=4.5, center=<-9.2054 -0.6174 -0.4386>, rotation=<rot 3.3802 4.1386 59.3148>}]
 dyn_frill_bck_r [{shapeType=Box, extentX=19, extentY=14.5, extentZ=5, center=<9.2891 0.1477 0.1595>, rotation=<rot 3.4211 4.2078 -120.6811>}]
 dyn_frill_frt [{shapeType=Box, extentX=13.5, extentY=4, extentZ=20.5, center=<0 1.15 -9>, rotation=<rot 0 0 -3.1066>}]
 dyn_frill_frt_l [{shapeType=Box, extentX=21.5, extentY=14, extentZ=6.5, center=<-12 3.5 -1>, rotation=<rot 3.6 -5.5 -80>}]
 dyn_frill_frt_r [{shapeType=Box, extentX=22, extentY=15, extentZ=5, center=<12 -1 1>, rotation=<rot 3.6 -12 100>}]
 dyn_frill_side_l [{shapeType=Box, extentX=18, extentY=14.5, extentZ=4, center=<-8.5 0 -3.75>, rotation=<rot 24 1 4>}]
 dyn_frill_side_r [{shapeType=Box, extentX=18, extentY=14.5, extentZ=4, center=<8.7 0 3.75>, rotation=<rot 24.5 1 -176>}]
 pelvis [{shapeType=Capsule, length=10.8103, radius=8.1851, center=<-3.9396 0 0>}]
 sub_thigh_l [{shapeType=Capsule, length=0.505, radius=0.505, rotation=<rot 0 -90 180>}]
 sub_thigh_r [{shapeType=Capsule, length=0.505, radius=0.505, rotation=<rot 0 -90 0>}]
 thigh_l [{shapeType=Capsule, length=18.5, radius=8, center=<-21.5144 0.608 0.8317>, rotation=<rot 90 0 -179.9999>}]
 thigh_r [{shapeType=Capsule, length=20, radius=8, center=<20 0 0>, rotation=<rot 90 0 0>}]
}
constraints [
 {
  bone1Name=sub_thigh_r
  bone2Name=thigh_r
  swing1LimitDegrees=45
  swing1Motion=Free
  swing2LimitDegrees=45
  swing2Motion=Free
  twistLimitDegrees=45
  twistMotion=Free
 }
 {bone1Name=dyn_frill_bck, bone2Name=driver_frill_bck, swing1LimitDegrees=45, swing2LimitDegrees=45, twistLimitDegrees=45}
 {
  bone1Name=sub_thigh_l
  bone2Name=thigh_l
  swing1LimitDegrees=45
  swing1Motion=Free
  swing2LimitDegrees=45
  swing2Motion=Free
  twistLimitDegrees=45
  twistMotion=Free
 }
 {bone1Name=dyn_frill_side_l, bone2Name=driver_frill_side_l, swing1LimitDegrees=30, swing2LimitDegrees=30, twistLimitDegrees=10}
 {bone1Name=dyn_frill_frt_l, bone2Name=driver_frill_frt_l, swing1LimitDegrees=20, swing2LimitDegrees=20, twistLimitDegrees=10}
 {bone1Name=dyn_frill_bck_l, bone2Name=driver_frill_bck_l, swing1LimitDegrees=45, swing2LimitDegrees=45, twistLimitDegrees=15}
 {bone1Name=dyn_frill_bck, bone2Name=dyn_frill_bck_l, swing1LimitDegrees=45, swing2LimitDegrees=45, twistLimitDegrees=45}
 {bone1Name=dyn_frill_frt, bone2Name=Flap_Front_Driven_Bone, swing1LimitDegrees=45, swing2LimitDegrees=45, twistLimitDegrees=45}
 {
  bone1Name=thigh_r
  bone2Name=pelvis
  swing1LimitDegrees=45
  swing1Motion=Free
  swing2LimitDegrees=45
  swing2Motion=Free
  twistLimitDegrees=45
  twistMotion=Free
 }
 {
  bone1Name=driver_frill_bck
  bone2Name=pelvis
  swing1LimitDegrees=45
  swing1Motion=Free
  swing2LimitDegrees=45
  swing2Motion=Free
  twistLimitDegrees=45
  twistMotion=Free
 }
 {
  bone1Name=driver_frill_bck_r
  bone2Name=sub_thigh_r
  swing1LimitDegrees=45
  swing1Motion=Free
  swing2LimitDegrees=45
  swing2Motion=Free
  twistLimitDegrees=45
  twistMotion=Free
 }
 {
  bone1Name=driver_frill_side_r
  bone2Name=sub_thigh_r
  swing1LimitDegrees=45
  swing1Motion=Free
  swing2LimitDegrees=45
  swing2Motion=Free
  twistLimitDegrees=45
  twistMotion=Free
 }
 {
  bone1Name=driver_frill_frt_r
  bone2Name=sub_thigh_r
  swing1LimitDegrees=45
  swing1Motion=Free
  swing2LimitDegrees=45
  swing2Motion=Free
  twistLimitDegrees=45
  twistMotion=Free
 }
 {
  bone1Name=Flap_Front_Driven_Bone
  bone2Name=pelvis
  swing1LimitDegrees=45
  swing1Motion=Free
  swing2LimitDegrees=45
  swing2Motion=Free
  twistLimitDegrees=45
  twistMotion=Free
 }
 {
  bone1Name=driver_frill_bck_l
  bone2Name=sub_thigh_l
  swing1LimitDegrees=45
  swing1Motion=Free
  swing2LimitDegrees=45
  swing2Motion=Free
  twistLimitDegrees=45
  twistMotion=Free
 }
 {
  bone1Name=driver_frill_side_l
  bone2Name=sub_thigh_l
  swing1LimitDegrees=45
  swing1Motion=Free
  swing2LimitDegrees=45
  swing2Motion=Free
  twistLimitDegrees=45
  twistMotion=Free
 }
 {
  bone1Name=driver_frill_frt_l
  bone2Name=sub_thigh_l
  swing1LimitDegrees=45
  swing1Motion=Free
  swing2LimitDegrees=45
  swing2Motion=Free
  twistLimitDegrees=45
  twistMotion=Free
 }
 {bone1Name=thigh_l, bone2Name=pelvis, swing1LimitDegrees=45, swing2LimitDegrees=45, twistLimitDegrees=45}
 {bone1Name=dyn_frill_frt_r, bone2Name=driver_frill_frt_r, swing1LimitDegrees=20, swing2LimitDegrees=20, twistLimitDegrees=10}
 {bone1Name=dyn_frill_side_r, bone2Name=driver_frill_side_r, swing1LimitDegrees=30, swing2LimitDegrees=30, twistLimitDegrees=10}
 {bone1Name=dyn_frill_bck_r, bone2Name=driver_frill_bck_r, swing1LimitDegrees=45, swing2LimitDegrees=45, twistLimitDegrees=15}
 {bone1Name=dyn_frill_bck, bone2Name=dyn_frill_bck_r, swing1LimitDegrees=45, swing2LimitDegrees=45, twistLimitDegrees=45}
 {bone1Name=dyn_frill_side_r, bone2Name=dyn_frill_frt_r, swing1LimitDegrees=45, swing2LimitDegrees=45, twistLimitDegrees=45}
 {bone1Name=dyn_frill_side_l, bone2Name=dyn_frill_frt_l, swing1LimitDegrees=45, swing2LimitDegrees=45, twistLimitDegrees=45}
]
