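# Physics asset for the strap on the F_MED_Elf female medium player body, covering a small dynamic bone chain at the pelvis and right thigh.
# Capsule bodies sit on the pelvis, right thigh, strap driver and strap bone, which carries two capsules.
# The thigh and driver connect to the pelvis with free angular motion, while the strap hangs off the driver with narrow swing and very tight twist limits.
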PhysicsAsset /BRCosmetics/Characters/Player/Female/Medium/Bodies/F_MED_Elf/Meshes/F_MED_Elf_Strap
bodies {
 dyn_driver_strap_01 [{shapeType=Capsule, length=1, radius=1}]
 dyn_strap_01 [
  {shapeType=Capsule, length=2.5, radius=1.75, center=<-1 0 0>, rotation=<rot 90 0 0>}
  {shapeType=Capsule, length=9, radius=8.5, center=<-13.5553 -0.0686 -1.4777>, rotation=<rot -51.4288 0 -87.2735>}
 ]
 pelvis [{shapeType=Capsule, length=6.4689, radius=3.0903}]
 thigh_r [{shapeType=Capsule, length=17, radius=8.4048, center=<7.5384 0.8695 0.6289>, rotation=<rot -0.4897 -88.6883 -106.4654>}]
}
constraints [
 {
  bone1Name=thigh_r
  bone2Name=pelvis
  swing1LimitDegrees=45
  swing1Motion=Free
  swing2LimitDegrees=45
  swing2Motion=Free
  twistLimitDegrees=45
  twistMotion=Free
 }
 {
  bone1Name=dyn_driver_strap_01
  bone2Name=pelvis
  swing1LimitDegrees=45
  swing1Motion=Free
  swing2LimitDegrees=45
  swing2Motion=Free
  twistLimitDegrees=45
  twistMotion=Free
 }
 {bone1Name=dyn_strap_01, bone2Name=dyn_driver_strap_01, swing1LimitDegrees=10, swing2LimitDegrees=25, twistLimitDegrees=2}
]
