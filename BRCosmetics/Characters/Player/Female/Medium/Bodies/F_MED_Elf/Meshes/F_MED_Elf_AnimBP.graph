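# Cosmetic Anim Blueprint for the female medium Elf body that adds secondary motion to skirt frills, scarf and straps.
# It copies the attached parent's pose, then chains bone constraints, additive offsets and RigidBody simulations, one borrowing Future Pink's physics asset.
# The update graph derives offset alphas from thigh and head twist and skydiving state.
# It also computes crouch, jump, DBNO and side-movement values that the AnimGraph never reads.
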AnimBlueprint /BRCosmetics/Characters/Player/Female/Medium/Bodies/F_MED_Elf/Meshes/F_MED_Elf_AnimBP
target_skeleton=&/F_MED_Elf_Skeleton
graphs [
 AnimGraph {nodes [
   n1 = "Constraint - sub_thigh_l (P)"(ComponentPose=Local To Component(LocalPose=Copy Pose From Mesh() {bUseAttachedParent=true}), Alpha=1, ConstraintWeights_0=1) {
    alphaBoolBlend {blendOption=Linear}
    boneToModify {boneName=sub_thigh_l}
    constraintSetup [{targetBone {boneName=thigh_twist_01_l}}]
    constraintWeights [1]
   }
   n2 = "Constraint - sub_thigh_r (P)"(ComponentPose=@n1, Alpha=1, ConstraintWeights_0=1) {
    alphaBoolBlend {blendOption=Linear}
    boneToModify {boneName=sub_thigh_r}
    constraintSetup [{targetBone {boneName=thigh_twist_01_r}}]
    constraintWeights [1]
   }
   n3 = Get legUp_L() -> (legUp_L:double)
   n4 = "Transform (Modify) Bone - Bone: driver_frill_frt_l"(ComponentPose=@n2, Alpha=@n3, Translation="4.841,23.026,9.924", Rotation="-18,30,12", Scale="1,1,1") {
    rotationMode=BMM_Additive
    rotationSpace=BCS_BoneSpace
    translationMode=BMM_Additive
    translationSpace=BCS_BoneSpace
    alphaBoolBlend {blendOption=Linear}
    alphaScaleBiasClamp {interpSpeedDecreasing=15, interpSpeedIncreasing=15}
    boneToModify {boneName=driver_frill_frt_l}
    rotation=<rot -18 30 12>
    translation=<4.841 23.026 9.924>
   }
   n5 = "Transform (Modify) Bone - Bone: driver_frill_side_l"(ComponentPose=@n4, Alpha=@n3, Translation="5,0,0", Rotation="-15,0,0", Scale="1,1,1") {
    rotationMode=BMM_Additive
    rotationSpace=BCS_BoneSpace
    scaleMode=BMM_Additive
    scaleSpace=BCS_BoneSpace
    translationMode=BMM_Additive
    translationSpace=BCS_BoneSpace
    alphaBoolBlend {blendOption=Linear}
    alphaScaleBiasClamp {interpSpeedDecreasing=15, interpSpeedIncreasing=15, inRange {max=0, min=-100}, outRange {max=0, min=1}}
    boneToModify {boneName=driver_frill_side_l}
    rotation {pitch=-15}
    translation {x=5}
   }
   n6 = "Transform (Modify) Bone - Bone: dyn_frill_frt"(ComponentPose=@n5, Alpha=@n3, Translation="0,0,10", Rotation="0,0,-20", Scale="1,1,1") {
    rotationMode=BMM_Additive
    rotationSpace=BCS_BoneSpace
    translationMode=BMM_Additive
    translationSpace=BCS_BoneSpace
    alphaBoolBlend {blendOption=Linear}
    alphaScaleBiasClamp {interpSpeedDecreasing=15, interpSpeedIncreasing=15}
    boneToModify {boneName=dyn_frill_frt}
    rotation {roll=-20}
    translation {z=10}
   }
   n7 = "Transform (Modify) Bone - Bone: driver_frill_frt_r"(
    ComponentPose=@n6
    Alpha=Get legUp_R() -> (legUp_R:double)
    Translation="-10.008,-4.250,-6.153"
    Rotation="0.000,-5.000,0.000"
    Scale="1,1,1"
   ) {
    rotationMode=BMM_Additive
    rotationSpace=BCS_BoneSpace
    translationMode=BMM_Additive
    translationSpace=BCS_BoneSpace
    alphaBoolBlend {blendOption=Linear}
    alphaScaleBiasClamp {interpSpeedDecreasing=15, interpSpeedIncreasing=15, inRange {max=0, min=-100}, outRange {max=0, min=1}}
    boneToModify {boneName=driver_frill_frt_r}
    rotation=<rot 0 -5 0>
    translation=<-10.008 -4.25 -6.153>
   }
   n8 = "Constraint - Flap_Front_Driven_Bone (PP)"!(ComponentPose=@n7, Alpha=1, ConstraintWeights_0=0.5, ConstraintWeights_1:float=0.5) -> (Pose) {
    alphaBoolBlend {blendOption=Linear}
    boneToModify {boneName=Flap_Front_Driven_Bone}
    constraintSetup [{targetBone {boneName=dyn_frill_frt_l}}, {targetBone {boneName=dyn_frill_frt_r}}]
    constraintWeights [1,1]
   }
   n9 = "Constraint - driver_frill_bck (PP)"!(ComponentPose=@n8, Alpha=1, ConstraintWeights_0=0.5, ConstraintWeights_1:float=0.5) -> (Pose) {
    alphaBoolBlend {blendOption=Linear}
    boneToModify {boneName=driver_frill_bck}
    constraintSetup [{targetBone {boneName=dyn_frill_bck_l}}, {targetBone {boneName=dyn_frill_bck_r}}]
    constraintWeights [1,1]
   }
   n10 = RigidBody(ComponentPose=@n9) {
    bClampLinearTranslationLimitToRefPose=true
    bOverrideWorldGravity=true
    overridePhysicsAsset=&/F_MED_Elf_Skirt_L_Physics
    simulationSpace=BaseBoneSpace
    alphaBoolBlend {blendOption=Linear}
    baseBoneRef {boneName=thigh_l}
    componentAppliedLinearAccClamp=<3000 3000 3000>
    componentLinearAccScale=<1 1 1>
    componentLinearVelScale=<1 1 1>
    overrideWorldGravity {z=-980}
   }
   n11 = RigidBody(ComponentPose=@n10) {
    bClampLinearTranslationLimitToRefPose=true
    bOverrideWorldGravity=true
    overridePhysicsAsset=&/F_MED_Elf_Skirt_R_Physics
    simulationSpace=BaseBoneSpace
    alphaBoolBlend {blendOption=Linear}
    baseBoneRef {boneName=thigh_r}
    componentAppliedLinearAccClamp=<3000 3000 3000>
    componentLinearAccScale=<1 1 1>
    componentLinearVelScale=<1 1 1>
    overrideWorldGravity {z=-980}
   }
   n12 = RigidBody(ComponentPose=@n11) {
    bOverrideWorldGravity=true
    overridePhysicsAsset=/BRCosmetics/Characters/Player/Female/Medium/Bodies/F_MED_Future_Pink/Meshes/F_MED_Future_Pink_Physics
    simulationSpace=BaseBoneSpace
    alphaBoolBlend {blendOption=Linear}
    baseBoneRef {boneName=pelvis}
    componentLinearAccScale=<0.3 0.3 0.3>
    componentLinearVelScale=<0.6 0.6 0.6>
    overrideWorldGravity {z=-980}
   }
   n13 = Get isSkyDivingOn() -> (isSkyDivingOn:double)
   n14 = "Transform (Modify) Bone - Bone: dyn_frill_bck"(ComponentPose=@n12, Alpha=@n13, Translation="0,-3,0", Rotation="0,0,0", Scale="1,1,1") {
    rotationMode=BMM_Additive
    rotationSpace=BCS_BoneSpace
    translationMode=BMM_Additive
    translationSpace=BCS_BoneSpace
    alphaBoolBlend {blendOption=Linear}
    alphaScaleBiasClamp {bInterpResult=true, outRange {max=0, min=1}}
    boneToModify {boneName=dyn_frill_bck}
    translation {y=-3}
   }
   n15 = "Transform (Modify) Bone - Bone: dyn_frill_bck_l"(ComponentPose=@n14, Alpha=@n13, Translation="0,-3,0", Rotation="0,0,0", Scale="1,1,1") {
    rotationMode=BMM_Additive
    rotationSpace=BCS_BoneSpace
    translationMode=BMM_Additive
    translationSpace=BCS_BoneSpace
    alphaBoolBlend {blendOption=Linear}
    alphaScaleBiasClamp {bInterpResult=true, outRange {max=0, min=1}}
    boneToModify {boneName=dyn_frill_bck_l}
    translation {y=-3}
   }
   n16 = "Transform (Modify) Bone - Bone: dyn_frill_bck_r"(ComponentPose=@n15, Alpha=@n13, Translation="0,3,0", Rotation="0,0,0", Scale="1,1,1") {
    rotationMode=BMM_Additive
    rotationSpace=BCS_BoneSpace
    translationMode=BMM_Additive
    translationSpace=BCS_BoneSpace
    alphaBoolBlend {blendOption=Linear}
    alphaScaleBiasClamp {bInterpResult=true, outRange {max=0, min=1}}
    boneToModify {boneName=dyn_frill_bck_r}
    translation {y=3}
   }
   n17 = "Transform (Modify) Bone - Bone: dyn_driver_frnt_scarf"(ComponentPose=@n16, Alpha=Get scarf_alpha() -> (scarf_alpha:double), Translation="0,0,0", Rotation="5,0,0", Scale="1,1,1") {
    rotationMode=BMM_Additive
    rotationSpace=BCS_BoneSpace
    alphaBoolBlend {blendOption=Linear}
    alphaScaleBiasClamp {bInterpResult=true, interpSpeedDecreasing=30, interpSpeedIncreasing=30}
    boneToModify {boneName=dyn_driver_frnt_scarf}
    rotation {pitch=5}
   }
   n18 = "Transform (Modify) Bone - Bone: dyn_driver_bck_scarf"(ComponentPose=@n17, Alpha=Get scarf_bck_alpha() -> (scarf_bck_alpha:double), Translation="1,0,-2", Rotation="-8,0.000,0.000", Scale="1,1,1") {
    rotationMode=BMM_Additive
    rotationSpace=BCS_BoneSpace
    translationMode=BMM_Additive
    translationSpace=BCS_BoneSpace
    alphaBoolBlend {blendOption=Linear}
    alphaScaleBiasClamp {bInterpResult=true, interpSpeedDecreasing=30, interpSpeedIncreasing=30}
    boneToModify {boneName=dyn_driver_bck_scarf}
    rotation {pitch=-8}
    translation {x=1, z=-2}
   }
   n19 = "Constraint - dyn_parent_strap_frt_01 (PPPPP)"!(
    ComponentPose=@n18
    Alpha=1
    ConstraintWeights_0=0.0997
    ConstraintWeights_1:float=0.2558
    ConstraintWeights_2:float=0.0127
    ConstraintWeights_3:float=0.051
    ConstraintWeights_4:float=0.5807
   ) -> (Pose) {
    alphaBoolBlend {blendOption=Linear}
    boneToModify {boneName=dyn_parent_strap_frt_01}
    constraintSetup [
     {targetBone {boneName=spine_01}}
     {targetBone {boneName=spine_02}}
     {targetBone {boneName=spine_03}}
     {targetBone {boneName=spine_04}}
     {targetBone {boneName=pelvis}}
    ]
    constraintWeights [1,1,1,1,1]
   }
   n20 = "Constraint - dyn_parent_strap_bck_01 (PPPPP)"!(
    ComponentPose=@n19
    Alpha=1
    ConstraintWeights_0=0.2379
    ConstraintWeights_1:float=0.2985
    ConstraintWeights_2:float=0.0509
    ConstraintWeights_3:float=0.0139
    ConstraintWeights_4:float=0.3987
   ) -> (Pose) {
    alphaBoolBlend {blendOption=Linear}
    boneToModify {boneName=dyn_parent_strap_bck_01}
    constraintSetup [
     {targetBone {boneName=spine_01}}
     {targetBone {boneName=spine_02}}
     {targetBone {boneName=spine_03}}
     {targetBone {boneName=spine_04}}
     {targetBone {boneName=pelvis}}
    ]
    constraintWeights [1,1,1,1,1]
   }
   n21 = "Constraint - dyn_parent_strap_01 (PPPPP)"!(
    ComponentPose=@n20
    Alpha=1
    ConstraintWeights_0=0.162
    ConstraintWeights_1:float=0.3353
    ConstraintWeights_2:float=0.0737
    ConstraintWeights_3:float=0.1017
    ConstraintWeights_4:float=0.3272
   ) -> (Pose) {
    alphaBoolBlend {blendOption=Linear}
    boneToModify {boneName=dyn_parent_strap_01}
    constraintSetup [
     {targetBone {boneName=spine_01}}
     {targetBone {boneName=spine_02}}
     {targetBone {boneName=spine_03}}
     {targetBone {boneName=spine_04}}
     {targetBone {boneName=pelvis}}
    ]
    constraintWeights [1,1,1,1,1]
   }
   n22 = RigidBody(ComponentPose=@n21, Alpha=0.55) {
    bOverrideWorldGravity=true
    overridePhysicsAsset=&/F_MED_Elf_Strap
    alphaBoolBlend {blendOption=Linear}
    baseBoneRef {boneName=root}
    componentAppliedLinearAccClamp=<1000 1000 1000>
    overrideWorldGravity {z=-980}
    simSpaceSettings {
     maxAngularVelocity=1000
     maxLinearAcceleration=8000
     maxLinearVelocity=100000
     velocityScaleZ=0.35
     worldAlpha=1
     externalLinearDragV {z=0.3}
    }
   }
   n23 = "Transform (Modify) Bone - Bone: dyn_frill_frt_r"(ComponentPose=@n22, Alpha=1, Translation="0,-5,0", Rotation="0,0,0", Scale="1,1,1") {
    rotationMode=BMM_Additive
    rotationSpace=BCS_BoneSpace
    translationMode=BMM_Additive
    translationSpace=BCS_BoneSpace
    alphaBoolBlend {blendOption=Linear}
    alphaScaleBiasClamp {interpSpeedDecreasing=15, interpSpeedIncreasing=15}
    boneToModify {boneName=dyn_frill_frt_r}
    translation {y=-5}
   }
   n24 = "Transform (Modify) Bone - Bone: driver_frill_frt_l"!(ComponentPose=@n23, Alpha=1, Translation="0,2,0", Rotation="0.000,5.000,-0.000") -> (Pose) {
    rotationMode=BMM_Additive
    rotationSpace=BCS_BoneSpace
    translationMode=BMM_Additive
    translationSpace=BCS_BoneSpace
    alphaBoolBlend {blendOption=Linear}
    alphaScaleBiasClamp {interpSpeedDecreasing=15, interpSpeedIncreasing=15}
    boneToModify {boneName=driver_frill_frt_l}
    rotation=<rot 0 5 0>
    translation {y=2}
   }
   n25 = "Transform (Modify) Bone - Bone: driver_frill_side_l"!(ComponentPose=@n24, Alpha=1, Translation="0,0,-8", Rotation="-5.000,-0.000,-0.000") -> (Pose) {
    rotationMode=BMM_Additive
    rotationSpace=BCS_BoneSpace
    translationMode=BMM_Additive
    translationSpace=BCS_BoneSpace
    alphaBoolBlend {blendOption=Linear}
    alphaScaleBiasClamp {interpSpeedDecreasing=15, interpSpeedIncreasing=15}
    boneToModify {boneName=driver_frill_side_l}
    rotation=<rot -5 0 0>
    translation {z=-8}
   }
   n26 = "Transform (Modify) Bone - Bone: driver_frill_side_l"(ComponentPose=@n25, Alpha=1, Translation="0,0,-1.5", Rotation="0,0,0") {
    rotationMode=BMM_Additive
    rotationSpace=BCS_BoneSpace
    translationMode=BMM_Additive
    translationSpace=BCS_BoneSpace
    alphaBoolBlend {blendOption=Linear}
    alphaScaleBiasClamp {bInterpResult=true, interpSpeedDecreasing=5, interpSpeedIncreasing=7}
    boneToModify {boneName=driver_frill_side_l}
    translation {z=-1.5}
   }
   Output Pose(Result=Component To Local(ComponentPose=@n26)) {layerGroup=None}
  ]}
 EventGraph {nodes [
   Event BlueprintUpdateAnimation()
   > Get!(execute:Exec) -> (then:Exec, else:Exec, Body_Mesh:Skeletal Mesh Component Object Reference)
   > n1 = Sequence!(execute) -> (then_0, then_1, then_2:Exec, then_3:Exec, then_4:Exec)
   > n2 = Branch(Condition=Get bIsCrouching())
   n3 = Get PawnSpeed2D()
   > n4 = Branch(Condition="float <= float"!(A:double=MapRangeClamped(Value=@n3, InRangeA=0, InRangeB=150), B:double=0) -> (ReturnValue:bool) {is=title K2Node_CallFunction})
   > Set crouch_speed_alpha(crouch_speed_alpha:double=1) -> (Output_Get:double)
   > Set crouch_moving(crouch_moving:bool=false) -> (Output_Get:bool)
   > Set crouch_rigid_alpha(crouch_rigid_alpha:double=0.2) -> (Output_Get:double)
   > Set move_speed_alpha(move_speed_alpha:double=0) -> (Output_Get:double)
   Set crouch_speed_alpha(execute=@n4.else, crouch_speed_alpha=0)
   > Set crouch_moving(crouch_moving=true)
   > Set crouch_rigid_alpha(crouch_rigid_alpha=0.55)
   > Set move_speed_alpha(move_speed_alpha=1)
   Set crouch_rigid_alpha(execute=@n2.else, crouch_rigid_alpha=0.8)
   n5 = Sequence(execute=@n1.then_4)
   n6 = TryGetFortPawn()
   Is Valid(exec=@n5.then_2, InputObject=@n6)
   n7 = Get bIsDBNO!(self:Fort Pawn Object Reference=@n6) -> (bIsDBNO)
   > Set IsDBNO(IsDBNO:bool=@n7) -> (Output_Get:bool)
   > Set NotDBNO(NotDBNO:bool=NOT Boolean(A=@n7)) -> (Output_Get:bool)
   > n8 = Sequence!(execute) -> (then_0, then_1)
   n9 = Get bIsCrouching()
   > Set crouchAlpha(crouchAlpha:double="To Float (Boolean)"(InBool=@n9)) -> (Output_Get:double)
   n10 = Branch(execute=@n8.then_1, Condition=@n9)
   n11 = Get PawnSpeed2D()
   > n12 = Branch(Condition="float <= float"(A=MapRangeClamped(Value=@n11, InRangeA=0, InRangeB=150), B=0))
   > Set crouch_speed_alpha(crouch_speed_alpha=1)
   > Set crouch_moving(crouch_moving=false)
   > Set crouch_rigid_alpha(crouch_rigid_alpha=0.2)
   > Set move_speed_alpha(move_speed_alpha=0)
   Set crouch_speed_alpha(execute=@n12.else, crouch_speed_alpha=0)
   > Set crouch_moving(crouch_moving=true)
   > Set crouch_rigid_alpha(crouch_rigid_alpha=0.55)
   > Set move_speed_alpha(move_speed_alpha=1)
   Set crouch_rigid_alpha(execute=@n10.else, crouch_rigid_alpha=0.8)
   n13 = GetOwnerDirection!(self:Custom Character Part Anim Instance Object Reference) -> (ReturnValue:float)
   n14 = Branch(
    execute=@n5.then_0
    Condition=AND Boolean(
     A="float > float"(A=@n13, B=-120) {is=title K2Node_CallFunction}
     B="float < float"!(A:double=@n13, B:double=-60) -> (ReturnValue:bool) {is=title K2Node_CallFunction}
    )
   )
   n15 = Get Body_Mesh()
   n16 = BreakTransform(InTransform=GetDeltaTransformFromRefPose(self=@n15, BoneName=thigh_r, BaseName=pelvis))
   > Set moveLeftLegCrossAlpha(moveLeftLegCrossAlpha:double=MapRangeClamped(Value=BreakRotator(InRot=@n16.Rotation).Pitch, InRangeA=-15, InRangeB=0, OutRangeA=1, OutRangeB=0)) -> (Output_Get:double)
   Set moveLeftLegCrossAlpha(execute=@n14.else, moveLeftLegCrossAlpha=0)
   n17 = Branch(execute=@n5.then_1, Condition=AND Boolean(A="float > float"(A=@n13, B=60), B="float < float"(A=@n13, B=120)))
   n18 = BreakTransform(InTransform=GetDeltaTransformFromRefPose(self=@n15, BoneName=thigh_l, BaseName=pelvis))
   > Set moveRighttLegCrossAlpha(moveRighttLegCrossAlpha:double=MapRangeClamped(Value=BreakRotator(InRot=@n18.Rotation).Pitch, InRangeA=-15, InRangeB=0, OutRangeA=1, OutRangeB=0)) -> (Output_Get:double)
   Set moveRighttLegCrossAlpha(execute=@n17.else, moveRighttLegCrossAlpha=0)
   n19 = "Absolute (Float)"(A=GetOwnerDirection())
   Set isMovingSide(execute=@n5.then_4, isMovingSide:bool=AND Boolean(A="float > float"(A=@n19, B=60), B="float < float"(A=@n19, B=120))) -> (Output_Get:bool)
   Set jump_alpha(execute=@n5.then_3, jump_alpha:double=MapRangeClamped(Value=Get PawnVelocityZ(), InRangeA=-700, InRangeB=0, OutRangeA=1, OutRangeB=0)) -> (Output_Get:double)
   > n20 = Branch()
   > Set isSkyDivingOn(isSkyDivingOn:double="To Float (Boolean)"(InBool=Get bIsSkyDiving())) -> (Output_Get:double)
   Set isSkyDivingOn(execute=@n20.else, isSkyDivingOn=0)
   Set legUp_L(
    execute=@n1.then_2
    legUp_L:double=GetAlphaFromReferenceTransformDelta(
     SkeletalMeshComponent=Get Body_Mesh()
     BoneName=thigh_l
     BaseName=pelvis
     DeltaComponentType=QuaternionTwist
     InputRange=(X=25,Y=120)
     TwistAxis="0,0,1"
    )
   ) -> (Output_Get:double)
   > Set legUp_R(legUp_R:double=GetAlphaFromReferenceTransformDelta(
     SkeletalMeshComponent=Get Body_Mesh()
     BoneName=thigh_r
     BaseName=pelvis
     DeltaComponentType=QuaternionTwist
     InputRange=(X=25,Y=120)
     TwistAxis="0,0,1"
    )) -> (Output_Get:double)
   n21 = Get Body_Mesh()
   Set scarf_alpha(
    execute=@n1.then_1
    scarf_alpha:double=GetAlphaFromReferenceTransformDelta(SkeletalMeshComponent=@n21, BoneName=head, BaseName=spine_05, DeltaComponentType=QuaternionTwist, InputRange=(X=0,Y=-25), TwistAxis="0,0,1")
   ) -> (Output_Get:double)
   > Set scarf_bck_alpha(scarf_bck_alpha:double=GetAlphaFromReferenceTransformDelta(SkeletalMeshComponent=@n21, BoneName=head, BaseName=spine_05, DeltaComponentType=QuaternionTwist, InputRange=(X=0,Y=45), TwistAxis="0,0,1")) -> (Output_Get:double)
   Event OnBeginCrouching()
   > Set is_crouching(is_crouching:bool=true) -> (Output_Get:bool)
   Event OnEndCrouching()
   > Set is_crouching(is_crouching=false)
   Event BlueprintInitializeAnimation()
   > Set Body_Mesh(Body_Mesh=GetOwnerSkeletalMeshComponent())
   > Set Pawn(Pawn:Fort Pawn Object Reference=TryGetFortPawn()) -> (Output_Get:Fort Pawn Object Reference)
   TryGetPawnOwner()
   SelectFloat(A=1, B=0, bPickA=OR Boolean(A=Get bIsJumping(), B=Get bIsFalling()))
  ]}
 Leg Side In Rev Rot FNC {nodes [
   n1 = Leg Side In Rev Rot FNC!() -> (then, legSideRotation:double=~, mult:double=~) {is=title K2Node_FunctionEntry}
   > n2 = Branch(Condition=Get isMovingSide() -> (isMovingSide:bool))
   > n3 = Branch(Condition="float < float"(A=@n1, B=0))
   > n4 = Set legSideRevRot(legSideRevRot:Rotator=MakeRotator(Pitch=float * float(A=@n1, B=@n1.mult))) -> (Output_Get:Rotator)
   Return Node!(execute=[@n6.then, @n4.then, @n5.then], legSideMotionRevRot:Rotator=@n4)
   n5 = Set legSideRevRot(execute=@n3.else, legSideRevRot="0, 0, 0")
   n6 = Set legSideRevRot(execute=@n2.else, legSideRevRot="0, 0, 0")
  ]}
]
variables [
 Body_Mesh {type=SkeletalMeshComponent}
 crouch_moving {type=boolean}
 crouch_rigid_alpha {type=number}
 crouch_speed_alpha {type=number}
 crouchAlpha {type=number}
 is_crouching {type=boolean}
 IsDBNO {type=boolean}
 isMovingSide {type=boolean}
 isSkyDivingOn {type=number}
 jump_alpha {type=number}
 legSide_L {type=number}
 legSide_R {type=number}
 legSideRevRot {type=Rotator}
 legUp_L {type=number}
 legUp_R {type=number}
 move_speed_alpha {type=number}
 moveLeftLegCrossAlpha {type=number}
 moveRighttLegCrossAlpha {type=number}
 NewVar_1 {type=Rotator}
 NewVar_2 {type=Rotator}
 NewVar_3 {type=Rotator}
 NotDBNO {type=boolean}
 Pawn {type=FortPawn, default=null}
 scarf_alpha {type=number}
 scarf_bck_alpha {type=number}
]
